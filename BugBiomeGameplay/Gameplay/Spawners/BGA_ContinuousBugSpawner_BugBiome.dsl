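; Damageable bug biome spawner that starts Living World swarm encounters from three tube points, with chance-based bombers and initial AI around it.
; Damage, proximity, linked spawners or a connected chest trigger spawning; a timer pauses it when no players are near and it is not recently damaged.
; Weak-point, explosive and near-weak-point pickaxe hits crit, driving damage multipliers and effects.
; Death drops loot, spawns a destroyed actor and releases the swarm encounter.
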(event EventBeginPlay
  (bind _connectedrewardactor (Variables|FortBugBiomeSpawner|GetConnectedRewardActor))
  (bind _output (Utilities|ResolveSoftReference _connectedrewardactor))
  (bind _outputdelegate (EventDispatchers|CreateEvent :!event Custom|StartSpawning))
  (bind _spherecollision (Variables|Spawning|GetSphereCollision))
  (bind (_spawnsettings_bautomaticallyfindrewardactor _spawnsettings_rewardactorradius _spawnsettings_connectedrewardstate _spawnsettings_btriggerfromplayerproximity) (Utilities|Struct|BreakBugBiomeSpawnerSettings (Variables|GetSpawnSettings_bAutomaticallyFindRewardActor)))
  ; Begin Play
  (EventDispatchers|BindEventtoOnDamaged (delegate Custom|SpawnerDamaged))
  (Utilities|FlowControl|Sequence
    (:then_0
      (CallFunction|InitSpawnPoints)
      (Networking|SwitchHasAuthority
        (:Authority
          (Class|FortAsyncActionCurrentPlaylistReady|PlaylistDataReadyAsync
            (:OnReady
              (Utilities|FlowControl|DelayUntilNextTick)
              (CallFunction|SpawnInitialAI))))))
    (:then_1
      (Utilities|IsValid _output
        (:Is%20Valid
          (bind _asalways_spawn_bug_chest_bug_biome (Utilities|Casting|CastToAlwaysSpawn_BugChest_BugBiome _output))
          (switch Utilities|FlowControl|Switch|SwitchonEBugSpawnerRewardTriggerState _spawnsettings_connectedrewardstate
            (:Opened
              (Default|BindEventtoOnOpened _asalways_spawn_bug_chest_bug_biome _outputdelegate))
            (:PlayerHasApproached
              (Default|AssignOnPlayerApproached _asalways_spawn_bug_chest_bug_biome _outputdelegate))
            (:AttackedOrOpened
              (Default|BindEventtoOnAttackedorOpened _asalways_spawn_bug_chest_bug_biome _outputdelegate))))))
    (:then_2
      (for _array_element (Variables|FortBugBiomeSpawner|GetLinkedSpawners)
        (if (Utilities|IsValid (Utilities|ResolveSoftReference _array_element) :!class /Script/Engine.KismetSystemLibrary)
          (Ability|Async|WaitGameplayEventToActor (Utilities|ResolveSoftReference _array_element) "(TagName=\"Event.Athena.BugBiome.Spawner.StartedSpawning\")"
            (:EventReceived
              (CallFunction|TriggerSpawn))))))
    (:then_3
      (if (and _spawnsettings_btriggerfromplayerproximity (ScalableFloat|GetValueasBool (Variables|Default|GetHF_EnableProximityTrigger)))
        (Utilities|IsValid _spherecollision
          (:Is%20Valid
            (Collision|AssignOnComponentBeginOverlap _spherecollision (delegate Custom|OnComponentBeginOverlap))))))))

(event Custom|SpawnerDamaged
  (params (DamagedActor :type Actor%20Object%20Reference) (Damage :type Float%20%28single-precision%29) (InstigatedBy :type Controller%20Object%20Reference) (DamageCauser :type Actor%20Object%20Reference) (HitLocation :type Vector) (FHitComponent :type Primitive%20Component%20Object%20Reference) (BoneName :type Name) (Momentum :type Vector))
  ; Begin Play
  (CallFunction|StartSpawning))

(event EventTriggerSpawn
  ; On Trigger Spawn
  (Utilities|FlowControl|DoOnce false
    (:Completed
      (Networking|SwitchHasAuthority
        (:Authority
          (CallFunction|StartSpawning))))))

(event Custom|OnComponentBeginOverlap
  (params (OverlappedComponent :type Primitive%20Component%20Object%20Reference) (OtherActor :type Actor%20Object%20Reference) (OtherComp :type Primitive%20Component%20Object%20Reference) (OtherBodyIndex :type Integer) (bFromSweep :type Boolean) (SweepResult :type Hit%20Result%20Structure%20%28by%20ref%29))
  ; On Player Detection Begin Overlap
  (bind _returnvalue (GameplayTags|HasMatchingGameplayTag(Message) OtherActor "(TagName=\"NPC.ProtoSwarm\")"))
  (if (not _returnvalue)
    (CallFunction|StartSpawning)
    (Development|PrintString "START SPAWNING" true true "(R=0.000000,G=0.660000,B=1.000000,A=1.000000)" 1e+01)))

(event Custom|One%20of%20my%20spawners%20died
  (params (DamagedActor :type Actor%20Object%20Reference) (Damage :type Float%20%28single-precision%29) (InstigatedBy :type Controller%20Object%20Reference) (DamageCauser :type Actor%20Object%20Reference) (HitLocation :type Vector) (FHitComponent :type Primitive%20Component%20Object%20Reference) (BoneName :type Name) (Momentum :type Vector))
  ; On Spawners Died Events
  (if (> Damage 5.0)
    (Variables|Default|SetSpawning true)))

(event Custom|StartSpawning
  ; Start Pawning
  (if (not (Variables|Default|GetSpawning))
    (Variables|Default|SetSpawning true)
    (Class|BPFunctionLibrarySpawners|BugSpawnersonSpawned self)
    (CallFunction|ExtraLogginInformation "Calling Spawn from Tubes")
    (CallFunction|SpawnfromTubes)
    (bind _returnvalue (Utilities|Time|SetTimerbyEvent (EventDispatchers|CreateEvent :!event Custom|CheckInactive) (ScalableFloat|GetValueatLevel (Variables|Default|GetCheckForInactivityTimer)) true))
    (Variables|InternalData|SetCheckForInactivityTimerRef _returnvalue)))

(event Custom|SpawnerDied
  (params (DamagedActor :type Actor%20Object%20Reference) (Damage :type Float%20%28single-precision%29) (InstigatedBy :type Controller%20Object%20Reference) (DamageCauser :type Actor%20Object%20Reference) (HitLocation :type Vector) (FHitComponent :type Primitive%20Component%20Object%20Reference) (BoneName :type Name) (Momentum :type Vector))
  ; On Spawners Died Events
  (bind _array_element (Utilities|Array|ForEachLoopwithBreak (Variables|Default|GetManagedSpawners))
    (:LoopBody
      (Actor|DestroyActor _array_element))))

(event Custom|StopSpawning
  ; Start Pawning
  (if (Variables|Default|GetSpawning)
    (Variables|Default|SetSpawning false)
    (Class|FortAthenaLivingWorldEncounterInstance|PauseEncounter (Variables|InternalData|GetSwarmEncounterRef))))

(event Custom|SpawnFromTubes
  ; Spawn AI!
  (Networking|SwitchHasAuthority
    (:Authority
      (Utilities|FlowControl|Sequence
        (:then_0
          (Utilities|IsValid (Variables|InternalData|GetSwarmEncounterRef)
            (:Is%20Valid
              (Class|FortAthenaLivingWorldEncounterInstance|ResumeEncounter (Variables|InternalData|GetSwarmEncounterRef))
              (Class|FortAthenaLivingWorldEncounterInstance|RequestEventGeneration (Variables|InternalData|GetSwarmEncounterRef)))
            (:Is%20Not%20Valid
              (CallFunction|EnableSpawnPoints true)
              (Utilities|FlowControl|DelayUntilNextTick)
              (bind _returnvalue (CallFunction|GetEncountersforSquadSize))
              (bind (_isvalid _returnvalue_2) (Default|StartActivityEncounterStaticPointProvider (Game|GetGameState) self (.SwarmEncounter (Utilities|Struct|BreakBugBiomeSpawnerEncounterSet _returnvalue)) (Variables|InternalData|GetLivingWorld_SpawnLocations)))
              (if _isvalid
                (bind _output_get (Variables|InternalData|SetSwarmEncounterRef _returnvalue_2))
                (EventDispatchers|BindEventtoOnEventSpawned _output_get (EventDispatchers|CreateEvent :!event Custom|OnSwarmEncounterEvent))
                (Class|FortAthenaLivingWorldEncounterInstance|RequestEventGeneration _output_get)
                (EventDispatchers|BindEventtoOnActorSpawned _output_get (delegate Custom|SwarmActorSpawned))
                (else
                  (Fortnite|FortShippingLog self ":: Tried to start an Encounter, but could not! Make sure the Encounter is defined properly on the asset AND in the Living World Manager!!" true))))))
        (:then_1
          (Utilities|FlowControl|DoOnce false
            (:Completed
              (if (Math|Random|RandomBoolwithWeight (ScalableFloat|GetValueatLevel (Variables|Default|GetBomberSpawnPercentageChance)))
                (CallFunction|SpawnBomber)))))))))

(event Custom|SwarmActorSpawned
  (params (EncounterInstance :type Fort%20Athena%20Living%20World%20Encounter%20Instance%20Object%20Reference) (SpawnedActor :type Actor%20Object%20Reference) (EventData :type Fort%20Athena%20Living%20World%20Event%20Data%20Object%20Reference))
  ; Spawn AI!
  (CallFunction|SpawnFXMulti EncounterInstance SpawnedActor EventData)
  (Default|CallOnSpawnedAnActor SpawnedActor))

(event Custom|SpawnInitialAI
  ; Try to spawn AI around the spawner when it is spawned in the world.
  (if (and (Math|Random|RandomBoolwithWeight (ScalableFloat|GetValueatLevel (Variables|Default|GetInitialSpawnPercentageChance))) (ScalableFloat|GetValueasBool (Variables|Spawning|InitialSpawn|GetShouldSpawnAIAroundSpawner)))
    (bind _returnvalue (CallFunction|GetEncountersforSquadSize))
    (if (Utilities|IsValidSoftObjectReference (Utilities|Struct|BreakBugBiomeSpawnerEncounterSet _returnvalue))
      (CallFunction|TrySpawnInitialSpawn)
      (else
        (Fortnite|FortShippingLog self ":: Trying to spawn AI at Load, but no encounter was defined! Please make sure to add an Encounter." true)))))

(event Custom|TrySpawnInitialSpawn
  ; Try to spawn AI around the spawner when it is spawned in the world.
  (bind _asfort_athena_living_world_volume (Utilities|Casting|CastToFortAthenaLivingWorldVolume (Variables|/Script/Engine.ChildActorComponent|GetChildActor (Variables|Default|GetLivingWorldVolume_ChildActor))))
  (bind _output_get (Variables|InternalData|SetLivingWorldVolume _asfort_athena_living_world_volume))
  (if (ScalableFloat|GetValueasBool (Variables|Spawning|InitialSpawn|GetShouldUseCustomEQSQuery))
    (Class|FortAthenaLivingWorldVolume|SetEQSQuery _output_get (Variables|Spawning|InitialSpawn|GetInitialSpawnEQSOverride))
    (else
      _))
  (EventDispatchers|BindEventtoOnPointProviderRegistered _output_get (delegate Custom|LivingWorldInitialSpawnVolumeReady))
  (Class|FortAthenaLivingWorldPointProviderInterface|EnablePointProvider_DEPRECATED _output_get))

(event Custom|LivingWorldInitialSpawnVolumeReady
  (params (PointProviderVolume :type Fort%20Athena%20Living%20World%20Volume%20Object%20Reference))
  ; Try to spawn AI around the spawner when it is spawned in the world.
  (Class|FortAthenaLivingWorldManager|RequestEventGeneration (Class|FortAthenaLivingWorldManager|GetLivingWorldManager))
  (bind _returnvalue (CallFunction|GetEncountersforSquadSize))
  (bind (_isvalid _returnvalue_2) (Default|StartActivityEncounter (Game|GetGameState) (Variables|InternalData|GetLivingWorldVolume) self (Utilities|Struct|BreakBugBiomeSpawnerEncounterSet _returnvalue)))
  (bind _output_get (Variables|InternalData|SetInitialSpawnEncounterRef _returnvalue_2))
  (EventDispatchers|BindEventtoOnLastActorDied _output_get (delegate Custom|LastInitialEncounterActorDied)))

(event Custom|LastInitialEncounterActorDied
  (params (EncounterInstance :type Fort%20Athena%20Living%20World%20Encounter%20Instance%20Object%20Reference))
  ; Try to spawn AI around the spawner when it is spawned in the world.
  (Class|BPFLLagerCommunFunctions|EndActivityEncounter (Game|GetGameState) (Variables|InternalData|GetInitialSpawnEncounterRef)))

(event EventOnDamagePlayEffects
  (params (Damage :type Float%20%28single-precision%29) (DamageTags :type Gameplay%20Tag%20Container%20Structure%20%28by%20ref%29) (Momentum :type Vector :default "0, 0, 0") (HitInfo :type Hit%20Result%20Structure%20%28by%20ref%29) (InstigatedBy :type Fort%20Pawn%20Object%20Reference) (DamageCauser :type Actor%20Object%20Reference) (EffectContext :type Gameplay%20Effect%20Context%20Handle%20Structure))
  (bind (_bblockinghit _binitialoverlap _time _distance _location _impactpoint _normal) (Collision|BreakHitResult HitInfo))
  (bind _returnvalue (Math|Conversions|RotationFromXVector _normal))
  ; Play FXs
  (if (Ability|EffectContext|IsCriticalHit EffectContext)
    (bind _returnvalue_2 (Niagara|SpawnSystematLocation "/BugBiomeGameplay/Assets/VFX/NS_BugSpawner_Impact_WeakPoint.NS_BugSpawner_Impact_WeakPoint" _location _returnvalue))
    (Effects|Components|ParticleSystem|SetVectorParameter _returnvalue_2 "ImpactFXDirection" _normal)
    (Audio|PlaySoundatLocation "/BugBiomeGameplay/Assets/Audio/MSS_SwarmSpawner_Impact.MSS_SwarmSpawner_Impact" _location)
    (if (Pawn|IsLocallyViewed InstigatedBy)
      (Audio|PlaySound2D "/BugBiomeGameplay/Assets/Audio/MSS_SwarmSpawner_HitNotify.MSS_SwarmSpawner_HitNotify")
      (CallFunction|DoDamageFX))
    (else
      (Niagara|SpawnSystematLocation "/BugBiomeGameplay/Assets/VFX/NS_BugSpawner_Impact_OuterShell.NS_BugSpawner_Impact_OuterShell" _location _returnvalue)
      (Audio|PlaySoundatLocation "/BugBiomeGameplay/Assets/Audio/MSS_SwarmSpawner_Impact_Light.MSS_SwarmSpawner_Impact_Light" _location))))

(event EventOnDeathPlayEffects
  (params (Damage :type Float%20%28single-precision%29) (DamageTags :type Gameplay%20Tag%20Container%20Structure%20%28by%20ref%29) (Momentum :type Vector :default "0, 0, 0") (HitInfo :type Hit%20Result%20Structure%20%28by%20ref%29) (InstigatedBy :type Fort%20Pawn%20Object%20Reference) (DamageCauser :type Actor%20Object%20Reference) (EffectContext :type Gameplay%20Effect%20Context%20Handle%20Structure))
  ; Play FXs
  (|CallParentFunction Damage DamageTags Momentum HitInfo InstigatedBy DamageCauser EffectContext))

(event EventOnDeathServer
  (params (Damage :type Float%20%28single-precision%29) (DamageTags :type Gameplay%20Tag%20Container%20Structure%20%28by%20ref%29) (Momentum :type Vector :default "0, 0, 0") (HitInfo :type Hit%20Result%20Structure%20%28by%20ref%29) (InstigatedBy :type Controller%20Object%20Reference) (DamageCauser :type Actor%20Object%20Reference) (EffectContext :type Gameplay%20Effect%20Context%20Handle%20Structure))
  ; OnServerDeath
  (Inventory|SpawnPickupInWorldWithLootTier (Variables|Default|GetLoot%20Tier%20Name) (Transformation|GetWorldLocation (Variables|Default|GetSpawnerAttachPoint01)))
  (Game|SpawnActorfromClass (select (Variables|Default|GetbIsInTunnel) (Variables|Default|GetDestroyedActorClassNoRVT) (Variables|Default|GetDestroyedActorClass)) (Transformation|GetActorTransform) "Undefined" "OverrideRootScale" self (Variables|Actor|GetInstigator))
  (Class|FortAthenaLivingWorldManager|ReleaseEncounter (Class|FortAthenaLivingWorldManager|GetLivingWorldManager) (Variables|InternalData|GetSwarmEncounterRef)))

(event Custom|SpawnBomber
  ; Spawn Bomber
  (bind _asfort_athena_living_world_volume (Utilities|Casting|CastToFortAthenaLivingWorldVolume (Variables|/Script/Engine.ChildActorComponent|GetChildActor (Variables|Default|GetLivingWorldVolume_BomberSpawn))))
  (bind _output_get (Variables|InternalData|SetLivingWorldVolume _asfort_athena_living_world_volume))
  (EventDispatchers|BindEventtoOnPointProviderRegistered _output_get (delegate Custom|SpawnBomberEncounter))
  (Class|FortAthenaLivingWorldPointProviderInterface|EnablePointProvider_DEPRECATED _output_get))

(event Custom|SpawnBomberEncounter
  (params (PointProviderVolume :type Fort%20Athena%20Living%20World%20Volume%20Object%20Reference))
  (bind _bomberencounterref (Variables|InternalData|GetBomberEncounterRef))
  ; Spawn Bomber
  (Class|FortAthenaLivingWorldManager|RequestEventGeneration (Class|FortAthenaLivingWorldManager|GetLivingWorldManager))
  (bind _returnvalue (CallFunction|GetEncountersforSquadSize))
  (Utilities|IsValid _bomberencounterref
    (:Is%20Valid
      (Class|FortAthenaLivingWorldEncounterInstance|ResumeEncounter _bomberencounterref)
      (Class|FortAthenaLivingWorldEncounterInstance|RequestEventGeneration _bomberencounterref))
    (:Is%20Not%20Valid
      (bind (_isvalid _returnvalue_2) (Default|StartActivityEncounter (Game|GetGameState) (Variables|InternalData|GetLivingWorldVolume) self (.BomberEncounter (Utilities|Struct|BreakBugBiomeSpawnerEncounterSet _returnvalue))))
      (bind _output_get (Variables|InternalData|SetBomberEncounterRef _returnvalue_2))
      (Class|FortAthenaLivingWorldEncounterInstance|RequestEventGeneration _output_get)
      (EventDispatchers|BindEventtoOnActorDied _output_get (delegate Custom|OnBomberDied)))))

(event Custom|OnBomberDied
  (params (EncounterInstance :type Fort%20Athena%20Living%20World%20Encounter%20Instance%20Object%20Reference) (DiedActor :type Actor%20Object%20Reference))
  ; Spawn Bomber
  (Class|BPFLLagerCommunFunctions|EndActivityEncounter (Game|GetGameState) (Variables|InternalData|GetBomberEncounterRef)) ; End the encounter after 1 Bomber Death
)

(event Custom|CheckInactive
  ; Check if spawned should go inactive
  (if (not (or (Variables|InternalData|GetDamagedRecently) (Utilities|Array|IsNotEmpty (Collision|GetOverlappingActors (Variables|Spawning|GetSphereCollision) "/Script/FortniteGame.FortPlayerPawnAthena" :!class /Script/Engine.PrimitiveComponent))))
    (CallFunction|StopSpawning)
    (Utilities|Time|ClearandInvalidateTimerbyHandle (Variables|InternalData|GetCheckForInactivityTimerRef))))

(event EventOnDamageServer
  (params (Damage :type Float%20%28single-precision%29) (DamageTags :type Gameplay%20Tag%20Container%20Structure%20%28by%20ref%29) (Momentum :type Vector :default "0, 0, 0") (HitInfo :type Hit%20Result%20Structure%20%28by%20ref%29) (InstigatedBy :type Controller%20Object%20Reference) (DamageCauser :type Actor%20Object%20Reference) (EffectContext :type Gameplay%20Effect%20Context%20Handle%20Structure))
  (bind _damagedrecentlytimerref (Variables|InternalData|GetDamagedRecentlyTimerRef))
  ; Track if spawner has been damaged recently
  (|CallParentFunction Damage DamageTags Momentum HitInfo InstigatedBy DamageCauser EffectContext)
  (Variables|InternalData|SetDamagedRecently true)
  (if (Utilities|Time|IsValidTimerHandle _damagedrecentlytimerref)
    (Utilities|Time|ClearandInvalidateTimerbyHandle _damagedrecentlytimerref)
    (else
      _))
  (bind _returnvalue (Utilities|Time|SetTimerbyEvent (delegate Custom|ResetDamagedRecently) (ScalableFloat|GetValueatLevel (Variables|Default|GetDamagedRecentlyTimer))))
  (Variables|InternalData|SetDamagedRecentlyTimerRef _returnvalue))

(event Custom|ResetDamagedRecently
  ; Track if spawner has been damaged recently
  (Variables|InternalData|SetDamagedRecently false))

(event Custom|OnSwarmEncounterEvent
  (params (EncounterInstance :type Fort%20Athena%20Living%20World%20Encounter%20Instance%20Object%20Reference) (SpawnedActors :type Array%20of%20Actor%20Object%20References))
  ; Do things here when we spawn swarms
  (CallFunction|DoSpawnFX))

(event Custom|DoSpawnFX
  ; Cosmetics when bugs coming out of the membranes
  (Networking|CanExecuteCosmeticEvents
    (:True
      (if (Variables|Default|GetSpawning)
        (Rendering|Material|SetCustomPrimitiveDataVector4 (Variables|Default|GetSM_WeakPoints) 0 (Math|Vector4|MakeVector4 0.0 1.0 (Utilities|Time|GetGameTimeinSeconds)))))))

(event Custom|DoDamageFX
  ; Cosmetics when spawner is taking damage
  (Networking|CanExecuteCosmeticEvents
    (:True
      (Rendering|Material|SetCustomPrimitiveDataVector4 (Variables|Default|GetSM_WeakPoints) 0 (Math|Vector4|MakeVector4 0.0 1.0 0.0 (Utilities|Time|GetGameTimeinSeconds))))))

(event Custom|SpawnFX_Multi
  (params (EncounterInstance :type Fort%20Athena%20Living%20World%20Encounter%20Instance%20Object%20Reference) (SpawnedActor :type Actor%20Object%20Reference) (EventData :type Fort%20Athena%20Living%20World%20Event%20Data%20Object%20Reference))
  ; Spawn vfx
  (Networking|SwitchHasAuthority
    (:Remote
      (bind _closestspawnpoint (CallFunction|FindClosestSpawnPoint SpawnedActor))
      (Niagara|SpawnSystematLocation "/BugBiomeGameplay/Assets/VFX/NS_BugSpawner_BugEmerge.NS_BugSpawner_BugEmerge" (Transformation|GetActorLocation _closestspawnpoint) (Math|Conversions|RotationFromXVector (Transformation|GetActorForwardVector _closestspawnpoint)))
      (Audio|PlaySoundatLocation "/BugBiomeGameplay/Assets/Audio/MSS_SwarmSpawner_Spray.MSS_SwarmSpawner_Spray" (Transformation|GetActorLocation _closestspawnpoint)))))

(event Custom|ActiveFX
  (bind (_direction _alpha) (Timeline|ActivatedTimeline PlayFromStart)
    (:Update
      (Niagara|SetNiagaraVariable(Float) (Variables|Default|GetNS_BugSpawner_Idle_Healthy) "Activated" _alpha))))

(event Custom|InactiveFX
  (Timeline|ActivatedTimeline ReverseFromEnd))

(fn BlueprintModifyIncomingDamage
  (params (Damage :type Float%20%28single-precision%29) (DamageTags :type Gameplay%20Tag%20Container%20Structure%20%28by%20ref%29) (EffectContext :type Gameplay%20Effect%20Context%20Handle%20Structure%20%28by%20ref%29) (EventInstigator :type Controller%20Object%20Reference) (DamageCauser :type Actor%20Object%20Reference))
  (bind _returnvalue (Ability|EffectContext|GetHitResult EffectContext))
  (bind (_bblockinghit _binitialoverlap _time _distance _location _impactpoint _normal _impactnormal _physmat _hitactor _hitcomponent) (Collision|BreakHitResult _returnvalue))
  (bind (_bblockinghit_2 _binitialoverlap_2 _time_2 _distance_2 _location_2 _impactpoint_2) (Collision|BreakHitResult _returnvalue))
  (bind _shouldcrit (LocalVariables|GetShouldCrit))
  (|CallParentFunction Damage DamageTags EffectContext EventInstigator DamageCauser)
  (Utilities|FlowControl|Sequence
    (:then_0
      (if (GameplayTags|HasTag DamageTags "(TagName=\"Weapon.Melee.Impact.Pickaxe\")")
        (bind (_hitlocation _hitnormal _bonename _outhit _returnvalue_2) (Collision|SphereOverlapComponent (Variables|Default|GetSM_WeakPoints) _impactpoint_2 15.0 true false false)) ; If pickaxe do an additional sphere trace to see if we hit near a weakpoint
        (LocalVariables|SetShouldCrit _returnvalue_2)
        (else
          (LocalVariables|SetShouldCrit (or (== _hitcomponent (Variables|Default|GetSM_WeakPoints)) (GameplayTags|HasTag (Variables|GetDamageTags) "(TagName=\"Gameplay.Damage.Physical.Explosive\")"))))))
    (:then_1
      (Ability|EffectContext|SetCritical (Variables|GetEffectContext) _shouldcrit)
      (return (Math|Float|SelectFloat (* (Variables|GetDamage) (ScalableFloat|GetValueatLevel (Variables|Default|GetWeakpointDamageMultiplier))) (* (Variables|GetDamage) (ScalableFloat|GetValueatLevel (Variables|Default|GetShellDamageMultiplier))) _shouldcrit)))))

(fn EnableSpawnPoints
  (params (Enable :type Boolean :default true))
  (for _array_element (Variables|InternalData|GetLivingWorld_SpawnLocations)
    (bind _asbp_static_point_provider_thrower (Utilities|Casting|CastToBP_StaticPointProvider_Thrower _array_element))
    (if Enable
      (Class|FortAthenaLivingWorldPointProviderInterface|EnablePointProvider_DEPRECATED _asbp_static_point_provider_thrower)
      (else
        (Class|FortAthenaLivingWorldStaticPointProvider|DisablePointProvider _asbp_static_point_provider_thrower)))))

(fn ExtraLogginInformation
  (params (MessageToLog :type String))
  (Fortnite|FortShippingLog self (Utilities|String|ToString(Text) (Utilities|Text|FormatText "NSLOCTEXT(\"[F23B124C6BF16101BDBE3132644818E8]\", \"728CDA35405E931D1900D08450C121D5\", \":: SWARMSPAWNER :: {a}\")" MessageToLog)))
  (return))

(fn FindClosestSpawnPoint
  (params (Actor :type Actor%20Object%20Reference))
  (LocalVariables|SetActorSpawnLocation (Transformation|GetActorLocation Actor))
  (for _array_element (Variables|InternalData|GetLivingWorld_SpawnLocations)
    (bind _distance (Class|MissionBlueprintFunctionLibrary|DistanceBetweenTwoVectors (Transformation|GetActorLocation _array_element) (LocalVariables|GetActorSpawnLocation)))
    (if (< _distance (LocalVariables|GetClosestDistance))
      (LocalVariables|SetClosestDistance _distance)
      (LocalVariables|SetClosestSpawnerPoint _array_element)))
  (return (LocalVariables|GetClosestSpawnerPoint)))

(fn InitSpawnPoints
  (Variables|InternalData|SetLivingWorld_SpawnLocations (Utilities|Array|MakeArray (Utilities|Casting|CastToBP_StaticPointProvider_Thrower (Variables|/Script/Engine.ChildActorComponent|GetChildActor (Variables|Default|GetBP_StaticPointProvider_SpawnPoint01))) (Utilities|Casting|CastToBP_StaticPointProvider_Thrower (Variables|/Script/Engine.ChildActorComponent|GetChildActor (Variables|Default|GetBP_StaticPointProvider_SpawnPoint02))) (Utilities|Casting|CastToBP_StaticPointProvider_Thrower (Variables|/Script/Engine.ChildActorComponent|GetChildActor (Variables|Default|GetBP_StaticPointProvider_SpawnPoint03))))))

(fn OnRep_Spawning
  (if (Variables|Default|GetSpawning)
    (Rendering|Material|SetCustomPrimitiveDataVector4 (Variables|Default|GetSM_WeakPoints) 0 (Math|Vector4|MakeVector4 (Utilities|Time|GetGameTimeinSeconds) 1.0)) ; Activating
    (CallFunction|ActiveFX)
    (else
      (Rendering|Material|SetCustomPrimitiveDataVector4 (Variables|Default|GetSM_WeakPoints) 0 (Math|Vector4|MakeVector4 (Utilities|Time|GetGameTimeinSeconds))) ; Deactivating
      (CallFunction|InactiveFX))))

(fn ConstructionScript
  (bind _fx_weakpoint_a (Variables|Default|GetFX_Weakpoint_A))
  (bind _fx_weakpoint_d (Variables|Default|GetFX_Weakpoint_D))
  (bind _ns_bugspawner_idle_healthy (Variables|Default|GetNS_BugSpawner_Idle_Healthy))
  (bind _fx_weakpoint_c (Variables|Default|GetFX_Weakpoint_C))
  (bind _fx_weakpoint_b (Variables|Default|GetFX_Weakpoint_B))
  (bind _fx_weakpoint_e (Variables|Default|GetFX_Weakpoint_E))
  (|CallParentFunction)
  (Utilities|Struct|SetmembersinBugBiomeSpawnerSettings (Variables|Spawning|GetSpawnSettings) "AttackedOrOpened")
  (Niagara|NiagaraSetPositionArray _ns_bugspawner_idle_healthy "FX_WeakpointPositions" (Utilities|Array|MakeArray (Transformation|GetWorldLocation _fx_weakpoint_a) (Transformation|GetWorldLocation _fx_weakpoint_b) (Transformation|GetWorldLocation _fx_weakpoint_c) (Transformation|GetWorldLocation _fx_weakpoint_d) (Transformation|GetWorldLocation _fx_weakpoint_e)))
  (Niagara|NiagaraSetVectorArray _ns_bugspawner_idle_healthy "FX_WeakpointDirections" (Utilities|Array|MakeArray (Transformation|GetForwardVector _fx_weakpoint_a) (Transformation|GetForwardVector _fx_weakpoint_b) (Transformation|GetForwardVector _fx_weakpoint_c) (Transformation|GetForwardVector _fx_weakpoint_d) (Transformation|GetForwardVector _fx_weakpoint_e)))
  (Niagara|SetNiagaraVariable(Float) _ns_bugspawner_idle_healthy "SmokeScale" 1.0))
