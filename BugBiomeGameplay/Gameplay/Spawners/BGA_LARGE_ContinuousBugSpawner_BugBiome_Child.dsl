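; Large continuous bug spawner variant for the Bug Biome, specializing a parent spawner Blueprint.
; InitSpawnPoints fills the living world spawn locations from seven child-actor thrower point providers.
; The construction script calls the parent, then passes world positions and forward vectors of five weakpoint components into array parameters on the healthy idle Niagara effect, along with an enlarged smoke scale.
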(fn InitSpawnPoints
  (Variables|InternalData|SetLivingWorld_SpawnLocations (Utilities|Array|MakeArray (Utilities|Casting|CastToBP_StaticPointProvider_Thrower (Variables|/Script/Engine.ChildActorComponent|GetChildActor (Variables|Default|GetBP_StaticPointProvider_SpawnPoint01))) (Utilities|Casting|CastToBP_StaticPointProvider_Thrower (Variables|/Script/Engine.ChildActorComponent|GetChildActor (Variables|Default|GetBP_StaticPointProvider_SpawnPoint02))) (Utilities|Casting|CastToBP_StaticPointProvider_Thrower (Variables|/Script/Engine.ChildActorComponent|GetChildActor (Variables|Default|GetBP_StaticPointProvider_SpawnPoint03))) (Utilities|Casting|CastToBP_StaticPointProvider_Thrower (Variables|/Script/Engine.ChildActorComponent|GetChildActor (Variables|Default|GetBP_StaticPointProvider_SpawnPoint04))) (Utilities|Casting|CastToBP_StaticPointProvider_Thrower (Variables|/Script/Engine.ChildActorComponent|GetChildActor (Variables|Default|GetBP_StaticPointProvider_SpawnPoint05))) (Utilities|Casting|CastToBP_StaticPointProvider_Thrower (Variables|/Script/Engine.ChildActorComponent|GetChildActor (Variables|Default|GetBP_StaticPointProvider_SpawnPoint06))) (Utilities|Casting|CastToBP_StaticPointProvider_Thrower (Variables|/Script/Engine.ChildActorComponent|GetChildActor (Variables|Default|GetBP_StaticPointProvider_SpawnPoint07))))))

(fn ConstructionScript
  (bind _fx_weakpoint_a (Variables|Default|GetFX_Weakpoint_A))
  (bind _fx_weakpoint_d (Variables|Default|GetFX_Weakpoint_D))
  (bind _ns_bugspawner_idle_healthy (Variables|Default|GetNS_BugSpawner_Idle_Healthy))
  (bind _fx_weakpoint_c (Variables|Default|GetFX_Weakpoint_C))
  (bind _fx_weakpoint_b (Variables|Default|GetFX_Weakpoint_B))
  (bind _fx_weakpoint_e (Variables|Default|GetFX_Weakpoint_E))
  (|CallParentFunction)
  (Niagara|NiagaraSetPositionArray _ns_bugspawner_idle_healthy "FX_WeakpointPositions" (Utilities|Array|MakeArray (Transformation|GetWorldLocation _fx_weakpoint_a) (Transformation|GetWorldLocation _fx_weakpoint_b) (Transformation|GetWorldLocation _fx_weakpoint_c) (Transformation|GetWorldLocation _fx_weakpoint_d) (Transformation|GetWorldLocation _fx_weakpoint_e)))
  (Niagara|NiagaraSetVectorArray _ns_bugspawner_idle_healthy "FX_WeakpointDirections" (Utilities|Array|MakeArray (Transformation|GetForwardVector _fx_weakpoint_a) (Transformation|GetForwardVector _fx_weakpoint_b) (Transformation|GetForwardVector _fx_weakpoint_c) (Transformation|GetForwardVector _fx_weakpoint_d) (Transformation|GetForwardVector _fx_weakpoint_e)))
  (Niagara|SetNiagaraVariable(Float) _ns_bugspawner_idle_healthy "SmokeScale" 2.0))
