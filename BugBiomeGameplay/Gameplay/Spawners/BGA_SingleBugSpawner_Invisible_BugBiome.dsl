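; Invisible BugBiome spawner actor that spawns one big or small bug egg point provider, or sometimes ammo instead of a small one.
; Linked spawner events, reward chest events, auto-start or death lead to a one-time authority-side SpawnActor, which checks Living World budgets before requesting event generation.
; Damage or death hatches and destroys the actor, with client-only effects.
; The optional proximity overlap handler checks state but takes no action.
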(event EventBeginPlay
  (bind _returnvalue (Transformation|GetActorScale3D))
  (bind _connectedrewardactor (Variables|FortBugBiomeSpawner|GetConnectedRewardActor))
  (bind _output (Utilities|ResolveSoftReference _connectedrewardactor))
  (bind _outputdelegate (EventDispatchers|CreateEvent :!event SpawnActor))
  (bind (_spawnsettings_bautomaticallyfindrewardactor _spawnsettings_rewardactorradius _spawnsettings_connectedrewardstate _spawnsettings_btriggerfromplayerproximity) (Utilities|Struct|BreakBugBiomeSpawnerSettings (Variables|GetSpawnSettings_bAutomaticallyFindRewardActor)))
  ; BeginPlay - Init world scale of eggs, setup events and setup linked spawners
  (Utilities|FlowControl|Sequence
    (:then_0
      (for _array_element (Variables|FortBugBiomeSpawner|GetLinkedSpawners)
        (Utilities|IsValid (Utilities|ResolveSoftReference _array_element)
          (:Is%20Valid
            (Ability|Async|WaitGameplayEventToActor (Utilities|ResolveSoftReference _array_element) "(TagName=\"Event.Athena.BugBiome.Spawner.StartedSpawning\")"
              (:EventReceived
                (CallFunction|TriggerSpawn)))))))
    (:then_1
      (Class|FortAsyncActionCurrentPlaylistReady|PlaylistDataReadyAsync
        (:OnReady
          (Variables|Default|SetEggDefaultScale _returnvalue)
          (Variables|Default|SetEggHatchScale (Math|Vector|MakeVector (+ (.x _returnvalue) 0.45) (+ (.y _returnvalue) 0.45) (+ (.z _returnvalue) 0.55)))
          (EventDispatchers|BindEventtoOnDied (delegate Custom|EggDied))
          (Utilities|IsValid _output
            (:Is%20Valid
              (bind _asalways_spawn_bug_chest_bug_biome (Utilities|Casting|CastToAlwaysSpawn_BugChest_BugBiome _output))
              (Default|BindEventtoOnOpened _asalways_spawn_bug_chest_bug_biome _outputdelegate)
              (Default|BindEventtoOnPlayerApproached _asalways_spawn_bug_chest_bug_biome _outputdelegate)
              (Default|BindEventtoOnAttackedorOpened _asalways_spawn_bug_chest_bug_biome _outputdelegate))))))
    (:then_2
      (if _spawnsettings_btriggerfromplayerproximity
        (Collision|AssignOnComponentBeginOverlap (Variables|Spawning|GetSphereCollision) (delegate Custom|OnComponentBeginOverlap_Event))))
    (:then_3
      (if (Variables|InstanceEditable|GetAutoSpawnAtStart)
        (CallFunction|TriggerSpawn)))))

(event Custom|EggDied
  (params (DamagedActor :type Actor%20Object%20Reference) (Damage :type Float%20%28single-precision%29) (InstigatedBy :type Controller%20Object%20Reference) (DamageCauser :type Actor%20Object%20Reference) (HitLocation :type Vector) (FHitComponent :type Primitive%20Component%20Object%20Reference) (BoneName :type Name) (Momentum :type Vector))
  ; BeginPlay - Init world scale of eggs, setup events and setup linked spawners
  (Variables|Default|SetEggWasDestroyed true))

(event Custom|OnComponentBeginOverlap_Event
  (params (OverlappedComponent :type Primitive%20Component%20Object%20Reference) (OtherActor :type Actor%20Object%20Reference) (OtherComp :type Primitive%20Component%20Object%20Reference) (OtherBodyIndex :type Integer) (bFromSweep :type Boolean) (SweepResult :type Hit%20Result%20Structure%20%28by%20ref%29))
  ; On Component Overlap - Egg Player Proximity Check Triggered
  (Networking|SwitchHasAuthority
    (:Authority
      (if (Utilities|Time|IsTimerActivebyHandle (Variables|Default|GetProximityTimer))
        _
        (elif (Variables|Default|GetEggHatchingInitiated)
          _)))))

(event EventTriggerSpawn
  ; On Trigger Spawn
  (Utilities|FlowControl|DoOnce false
    (:Completed
      (CallFunction|SpawnActor))))

(event Custom|HatchEgg
  ; On Egg Hatched
  (Utilities|FlowControl|DoOnce false
    (:Completed
      (Variables|Default|SetEggHatchingInitiated true)
      (CallFunction|EggHatchInitClientEffects)
      (Actor|DestroyActor))))

(event Custom|EggHatchInitClientEffects
  ; On Egg Hatched
  (if (not (Variables|Default|GetEggWasDestroyed))
    (Networking|SwitchHasAuthority
      (:Remote
        (Audio|SpawnSoundatLocation "/Game/Sounds/Fort_GamePlay_Sounds/Player/RiftPortal/RiftPortal_Expand_Cue.RiftPortal_Expand_Cue" (Transformation|GetActorLocation))))))

(event EventOnDamageServer
  (params (Damage :type Float%20%28single-precision%29) (DamageTags :type Gameplay%20Tag%20Container%20Structure%20%28by%20ref%29) (Momentum :type Vector :default "0, 0, 0") (HitInfo :type Hit%20Result%20Structure%20%28by%20ref%29) (InstigatedBy :type Controller%20Object%20Reference) (DamageCauser :type Actor%20Object%20Reference) (EffectContext :type Gameplay%20Effect%20Context%20Handle%20Structure))
  ; On Egg Damaged - Start to Hatch
  (Utilities|FlowControl|DoOnce false
    (:Completed
      (CallFunction|HatchEgg))))

(event EventOnDeathServer
  (params (Damage :type Float%20%28single-precision%29) (DamageTags :type Gameplay%20Tag%20Container%20Structure%20%28by%20ref%29) (Momentum :type Vector :default "0, 0, 0") (HitInfo :type Hit%20Result%20Structure%20%28by%20ref%29) (InstigatedBy :type Controller%20Object%20Reference) (DamageCauser :type Actor%20Object%20Reference) (EffectContext :type Gameplay%20Effect%20Context%20Handle%20Structure))
  ; On Egg Destroyed by Player
  (Utilities|FlowControl|DoOnce false
    (:Completed
      (CallFunction|HideEggmeshesonClient)
      (CallFunction|HatchEgg)
      (CallFunction|SpawnActor)
      (CallFunction|ClientDeathEffects))))

(event Custom|ClientDeathEffects
  ; Egg destruction FX
  (Networking|SwitchHasAuthority
    (:Remote
      (Components|Activation|Activate (Variables|Default|GetNS_Flopper_JellyFish_Explode) false)
      (Audio|SpawnSoundatLocation "/Game/Athena/Sounds/Items/Fizz/Paprika_Fizz_Explode_Cue.Paprika_Fizz_Explode_Cue" (Transformation|GetActorLocation)))))

(event Custom|HideEggmeshesOnClient
  ; Egg visibility on death
  (Variables|Default|SetEggVisibilityState false))

(fn SpawnActor
  (bind _spawntransform_location (Math|Transform|BreakTransform (Variables|GetSpawnTransform_Location)))
  (Class|BPFunctionLibrarySpawners|BugSpawnersonSpawned self)
  (Networking|SwitchHasAuthority
    (:Authority
      (if (not (Variables|Default|GetSpawned))
        (Variables|Default|SetSpawned true)
        (LocalVariables|SetSpawnTransform (Math|Transform|MakeTransform (+ (Transformation|GetWorldLocation (Variables|Spawning|GetSphereCollision)) (* (Math|Vector|GetUpVector (Transformation|GetActorRotation)) 5e+01)) (Math|Rotator|MakeRotator 0.0 0.0 (Math|Random|RandomFloatinRange -179.0 179.0))))
        (if (Variables|Default|GetAllowSpawnAI)
          (if (Variables|InstanceEditable|GetSpawnBigBug)
            (if (> (Class|FortAthenaLivingWorldManager|QueryEventBudget (Class|FortAthenaLivingWorldManager|GetLivingWorldManager) (Variables|Default|GetMediumLagerEntry) self) 0)
              (Game|SpawnActorfromClass "/ProtoSwarm/Blueprints/BP_PointProviderStatic_Egg_Medium.BP_PointProviderStatic_Egg_Medium_C" (LocalVariables|GetSpawnTransform) "AlwaysSpawn" "MultiplyWithRoot" 0 0))
            (elif (<= (Math|Random|RandomFloatinRange 0.0 1.0) (Variables|Default|GetChanceForLootOnSmall))
              (Inventory|SpawnPickupInWorldWithLootTier (Utilities|Select :Index (Math|Random|RandomIntegerinRange 0 4) :Option%200 "LTG_Vend_Item_Pistol_Ammo" :Option%201 "LTG_Vend_Item_Shotgun_Ammo" :Option%202 "LTG_Vend_Item_SMG_Ammo" :Option%203 "LTG_Vend_Item_Assault_Ammo" :Option%204 "LTG_Vend_Item_Sniper_Ammo") _spawntransform_location 0 0 false)
              (elif (> (Class|FortAthenaLivingWorldManager|QueryEventBudget (Class|FortAthenaLivingWorldManager|GetLivingWorldManager) (Variables|Default|GetSmallLagerEntry) self) 0)
                (Game|SpawnActorfromClass "/ProtoSwarm/Blueprints/BP_PointProviderStatic_Egg_Small.BP_PointProviderStatic_Egg_Small_C" (LocalVariables|GetSpawnTransform) "AlwaysSpawn" "MultiplyWithRoot" 0 0)
                (else
                  _))))
          (Class|FortAthenaLivingWorldManager|RequestEventGeneration (Class|FortAthenaLivingWorldManager|GetLivingWorldManager)))))))

(fn ConstructionScript
  (if (Variables|InstanceEditable|GetSpawnBigBug)
    (Transformation|SetRelativeScale3D (Variables|Default|GetStaticMesh) "2.500000,2.500000,2.500000")))
