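; Blueprint function library for Bug Biome spawners, with a shared handler for when a spawner spawns.
; BugSpawners_OnSpawned sends the StartedSpawning gameplay event to the given spawner actor.
; It then loops over that spawner's linked spawners, resolving each soft reference and calling TriggerSpawn on the valid ones, so a spawn propagates to linked spawners.
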(fn BugSpawners_OnSpawned
  (params (Self%20Actor :type Fort%20Bug%20Biome%20Spawner%20Object%20Reference))
  (Ability|SendGameplayEventtoActor Self%20Actor "(TagName=\"Event.Athena.BugBiome.Spawner.StartedSpawning\")")
  (for _array_element (Variables|/Script/BugBiomeGameplayRuntime.FortBugBiomeSpawner|GetLinkedSpawners Self%20Actor)
    (if (Utilities|IsValid (Utilities|ResolveSoftReference _array_element) :!class /Script/Engine.KismetSystemLibrary)
      (Class|FortBugBiomeSpawner|TriggerSpawn (Utilities|ResolveSoftReference _array_element)))))
