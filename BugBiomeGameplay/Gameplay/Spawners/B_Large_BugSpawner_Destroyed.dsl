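; Destroyed-state actor for the large bug spawner, playing its death effects and burst sound on BeginPlay.
; Scene component transforms fill Niagara arrays for idle drips and death holes, with Hole A entered twice.
; The death system and sound spawn at the destruction origin, here rather than in the live spawner for better alignment with this actor's spawn.
; Both death effects are scaled up, and the holes system activates last.
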(event EventBeginPlay
  (bind _fx_destructionorigin (Variables|Default|GetFX_DestructionOrigin))
  (bind _returnvalue (Transformation|GetWorldLocation _fx_destructionorigin))
  (bind _fx_hole_f (Variables|Default|GetFX_Hole_F))
  (bind _ns_bugspawner_death_holes (Variables|Default|GetNS_BugSpawner_Death_Holes))
  (bind _fx_hole_a (Variables|Default|GetFX_Hole_A))
  (bind _fx_hole_e (Variables|Default|GetFX_Hole_E))
  (bind _fx_hole_d (Variables|Default|GetFX_Hole_D))
  (bind _fx_hole_c (Variables|Default|GetFX_Hole_C))
  (bind _fx_hole_b (Variables|Default|GetFX_Hole_B))
  (Niagara|NiagaraSetPositionArray (Variables|Default|GetNS_BugSpawner_Idle_Destroyed) "FX_DropPositions" (Utilities|Array|MakeArray (Transformation|GetWorldLocation (Variables|Default|GetFX_Drips_A)) (Transformation|GetWorldLocation (Variables|Default|GetFX_Drips_B)) (Transformation|GetWorldLocation (Variables|Default|GetFX_Drips_C)) (Transformation|GetWorldLocation (Variables|Default|GetFX_Drips_D))))
  (Niagara|NiagaraSetPositionArray _ns_bugspawner_death_holes "FX_HolePositions" (Utilities|Array|MakeArray (Transformation|GetWorldLocation _fx_hole_a) (Transformation|GetWorldLocation _fx_hole_a) (Transformation|GetWorldLocation _fx_hole_b) (Transformation|GetWorldLocation _fx_hole_c) (Transformation|GetWorldLocation _fx_hole_d) (Transformation|GetWorldLocation _fx_hole_e) (Transformation|GetWorldLocation _fx_hole_f)))
  (Niagara|NiagaraSetVectorArray _ns_bugspawner_death_holes "FX_HoleDirections" (Utilities|Array|MakeArray (Transformation|GetForwardVector _fx_hole_a) (Transformation|GetForwardVector _fx_hole_a) (Transformation|GetForwardVector _fx_hole_b) (Transformation|GetForwardVector _fx_hole_c) (Transformation|GetForwardVector _fx_hole_d) (Transformation|GetForwardVector _fx_hole_e) (Transformation|GetForwardVector _fx_hole_f)))
  (Niagara|SetNiagaraVariable(Float) _ns_bugspawner_death_holes "HoleScale" 1.5)
  ; Spawn the destroyed FX from the bug spawner version here so it aligns better with this destroyed actor spawning
  (bind _returnvalue_2 (Niagara|SpawnSystematLocation "/BugBiomeGameplay/Assets/VFX/NS_BugSpawner_Death.NS_BugSpawner_Death" _returnvalue))
  (Niagara|SetNiagaraVariable(Float) _returnvalue_2 "SpawnerScale" 1.5)
  (Audio|PlaySoundatLocation "/BugBiomeGameplay/Assets/Audio/MSS_SwarmSpawner_Burst.MSS_SwarmSpawner_Burst" _returnvalue)
  (Components|Activation|Activate (Variables|Default|GetNS_BugSpawner_Death_Holes) false))

(fn ConstructionScript
  (|CallParentFunction))
